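; Anim notify state that locks player movement and jumping for the span of the animation it is placed on.
; On begin, it casts the mesh owner to FortPlayerPawn and applies the GE_BlockMove and GE_BlockJump gameplay effects through that pawn's ability system component.
; On end, it removes both effects by source effect, using the same component as instigator.
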(fn GetNotifyName
  (return "AnimStateNotify_LockPlayerMovement"))

(fn Received_NotifyBegin
  (params (MeshComp :type Skeletal%20Mesh%20Component%20Object%20Reference) (Animation :type Anim%20Sequence%20Base%20Object%20Reference) (TotalDuration :type Float%20%28single-precision%29) (EventReference :type Anim%20Notify%20Event%20Reference%20Structure%20%28by%20ref%29))
  (Utilities|IsValid MeshComp
    (:Is%20Valid
      (bind _asfort_player_pawn (Utilities|Casting|CastToFortPlayerPawn (Components|GetOwner MeshComp))
        (:then
          (Utilities|IsValid (Ability|GetFortAbilitySystemComponent _asfort_player_pawn)
            (:Is%20Valid
              (GameplayEffects|ApplyGameplayEffectToSelf (Ability|GetFortAbilitySystemComponent _asfort_player_pawn) "/Game/Athena/Items/Parents/MovementBlock/GE_BlockMove.GE_BlockMove_C" 0.0 (GameplayEffects|MakeEffectContext (Ability|GetFortAbilitySystemComponent _asfort_player_pawn)))
              (GameplayEffects|ApplyGameplayEffectToSelf (Ability|GetFortAbilitySystemComponent _asfort_player_pawn) "/Game/Athena/Items/Parents/MovementBlock/GE_BlockJump.GE_BlockJump_C" 0.0 (GameplayEffects|MakeEffectContext (Ability|GetFortAbilitySystemComponent _asfort_player_pawn)))
              (return true))
            (:Is%20Not%20Valid
              _)))
        (:CastFailed
          _))
      (return false))
    (:Is%20Not%20Valid
      _)))

(fn Received_NotifyEnd
  (params (MeshComp :type Skeletal%20Mesh%20Component%20Object%20Reference) (Animation :type Anim%20Sequence%20Base%20Object%20Reference) (EventReference :type Anim%20Notify%20Event%20Reference%20Structure%20%28by%20ref%29))
  (Utilities|IsValid MeshComp
    (:Is%20Valid
      (bind _asfort_player_pawn (Utilities|Casting|CastToFortPlayerPawn (Components|GetOwner MeshComp))
        (:then
          (Utilities|IsValid (Ability|GetFortAbilitySystemComponent _asfort_player_pawn)
            (:Is%20Valid
              (GameplayEffects|RemoveActiveGameplayEffectbySourceEffect (Ability|GetFortAbilitySystemComponent _asfort_player_pawn) "/Game/Athena/Items/Parents/MovementBlock/GE_BlockMove.GE_BlockMove_C" (Ability|GetFortAbilitySystemComponent _asfort_player_pawn) -1)
              (GameplayEffects|RemoveActiveGameplayEffectbySourceEffect (Ability|GetFortAbilitySystemComponent _asfort_player_pawn) "/Game/Athena/Items/Parents/MovementBlock/GE_BlockJump.GE_BlockJump_C" (Ability|GetFortAbilitySystemComponent _asfort_player_pawn) -1)
              (return true))
            (:Is%20Not%20Valid
              _)))
        (:CastFailed
          _))
      (return false))
    (:Is%20Not%20Valid
      _)))
