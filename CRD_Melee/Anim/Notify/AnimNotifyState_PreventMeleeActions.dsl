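; Anim notify state that suppresses melee actions on a Fortnite player pawn during its animation window through a configurable gameplay effect.
; Begin applies SuppressingGameplayEffect to the pawn's ability system; End removes it by source effect.
; Both first call IsLinkedAttack, which passes if LinkedAttack or the Creative_Melee_Parent weapon's current attack name is empty, or they match.
; This check keeps block suppression from dropping when switching between combo parts.
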(fn IsLinkedAttack
  (params (Pawn :type Fort%20Player%20Pawn%20Object%20Reference))
  ;[0 0.964 1] If the AnimNotify has a LinkedAttack specified, only start or end the GE if it is activated while the linkedattack is playing (or if attack is none). This is to stop the block disabling when you switch from one part of the combo to another.
  (if (== (Variables|Default|GetLinkedAttack) "")
    (return true)
    (else
      (bind _currentweapon (Variables|/Script/FortniteGame.FortPawn|GetCurrentWeapon Pawn)
        (:then
          (bind _ascreative_melee_parent (Utilities|Casting|CastToCreative_Melee_Parent _currentweapon)
            (:then
              (if (== (Variables|/CRD_Melee/Weapons/Creative_Melee_Parent.Creative_Melee_Parent_C|GetCurrentAttackName _ascreative_melee_parent) "")
                (return true)
                (else
                  (return (== (Variables|Default|GetLinkedAttack) (Variables|/CRD_Melee/Weapons/Creative_Melee_Parent.Creative_Melee_Parent_C|GetCurrentAttackName _ascreative_melee_parent))))))
            (:CastFailed
              (return false))))
        (:else)))))

(fn Received_NotifyBegin
  (params (MeshComp :type Skeletal%20Mesh%20Component%20Object%20Reference) (Animation :type Anim%20Sequence%20Base%20Object%20Reference) (TotalDuration :type Float%20%28single-precision%29) (EventReference :type Anim%20Notify%20Event%20Reference%20Structure%20%28by%20ref%29))
  (bind _asfort_player_pawn (Utilities|Casting|CastToFortPlayerPawn (Components|GetOwner MeshComp))
    (:then
      (if (CallFunction|IsLinkedAttack _asfort_player_pawn)
        (GameplayEffects|ApplyGameplayEffectToSelf (Ability|GetAbilitySystemComponent _asfort_player_pawn) (Variables|Default|GetSuppressingGameplayEffect) 0.0 (GameplayEffects|MakeEffectContext (Ability|GetAbilitySystemComponent _asfort_player_pawn)))
        (return true)
        (else
          _)))
    (:CastFailed
      _))
  (return false))

(fn Received_NotifyEnd
  (params (MeshComp :type Skeletal%20Mesh%20Component%20Object%20Reference) (Animation :type Anim%20Sequence%20Base%20Object%20Reference) (EventReference :type Anim%20Notify%20Event%20Reference%20Structure%20%28by%20ref%29))
  (bind _asfort_player_pawn (Utilities|Casting|CastToFortPlayerPawn (Components|GetOwner MeshComp))
    (:then
      (if (CallFunction|IsLinkedAttack _asfort_player_pawn)
        (GameplayEffects|RemoveActiveGameplayEffectbySourceEffect (Ability|GetAbilitySystemComponent _asfort_player_pawn) (Variables|Default|GetSuppressingGameplayEffect) 0 -1)
        (return true)
        (else
          _)))
    (:CastFailed
      _))
  (return false))
